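# Master material for Creative cube playground items, using default lit shading.
# It exposes scalar parameters for Metallic, Spec and Roughness, plus vector parameters for Color and Emissive, with mid-grey and black defaults.
# The graph holds only these parameters, with no textures or math, so cube variants can differ by parameter values alone.
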Material /Figment_S07/LM_7_00/Playgrounds/Items/Cubes/Materials/M_Creative_Cube_Master
graph {
 properties {pixelDepthOffsetMode=PDOM_Legacy, usedShadingModels=Default Lit}
 nodes [
  ScalarParameter() {parameterName=Metallic}
  ScalarParameter() {parameterName=Spec}
  ScalarParameter() {parameterName=Roughness}
  VectorParameter() {parameterName=Color, defaultValue=<rgba 0.5 0.5 0.5 0>}
  VectorParameter() {parameterName=Emissive, defaultValue=<rgba 0 0 0 1>}
 ]
}
